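# Default Lit material for a quest iris effect, built around a vertex offset that scales the mesh toward its object position.
# The Iris parameter, or a debug sine blended by DEBUG, scales the object-relative position between full collapse and no offset.
# A local Z-axis rotation is added with its angle fixed at zero, and vertex color red masks the sum.
# A test iris normal map is also sampled.
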Material /BattlepassS17/FX/Materials/M_Quest_Iris
graph {
 properties {pixelDepthOffsetMode=PDOM_Legacy, usedShadingModels=Default Lit}
 nodes [
  n1 = Subtract(A=WorldPosition().XYZ, B=ObjectPositionWS())
  n2 = LinearInterpolate(
   A=Constant() {r=-1}
   B=Constant()
   Alpha=LinearInterpolate(
    A=ScalarParameter() {defaultValue=1, parameterName=Iris}
    B=MaterialFunctionCall() {materialFunction=&DebugTimeSine}.Result
    Alpha=ScalarParameter() {parameterName=DEBUG}
   )
  )
  n3 = Transform(None=Constant3Vector() {constant=<rgba 0 0 1 0>}) {transformSourceType=TRANSFORMSOURCE_Local}
  n4 = Constant()
  n5 = Constant3Vector()
  Multiply(
   A=Add(A=Multiply(A=@n1, B=@n2), B=RotateAboutAxis(NormalizedRotationAxis=@n3, RotationAngle=@n4, PivotPoint=@n5, Position=@n1))
   B=VertexColor().R
  )
  TextureSample() {samplerType=SAMPLERTYPE_Normal, texture=/BattlepassS17/FX/Textures/N_Test_Iris}
 ]
}
